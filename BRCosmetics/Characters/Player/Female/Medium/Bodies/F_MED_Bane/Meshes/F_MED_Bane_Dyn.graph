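# Physics asset for the dynamic bones of the F_MED_Bane female player body, covering coat panels, zippers and a right pouch.
# Each box-shaped piece hinges on a tiny capsule MSTR anchor, with the front coat panels forming two-link chains.
# Coat and collar anchors connect to spine capsules with free angular motion, while ankle and pouch anchors have no parent constraint.
# The left zipper anchor locks to the lower left-front coat panel, so it rides that chain.
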PhysicsAsset /BRCosmetics/Characters/Player/Female/Medium/Bodies/F_MED_Bane/Meshes/F_MED_Bane_Dyn
bodies {
 dyn_ankle_zipper_l [{shapeType=Box, extentX=1.7732, extentY=0.983, extentZ=0.5001, center=<0.6923 0 0.2146>, rotation=<rot 14.9999 0 0>}]
 dyn_ankle_zipper_l_MSTR [{shapeType=Capsule, length=0.2, radius=0.2, rotation=<rot 90 0 0>}]
 dyn_ankle_zipper_r [{shapeType=Box, extentX=1.7732, extentY=0.983, extentZ=0.5001, center=<0.6923 0 0.2146>, rotation=<rot 14.9999 0 0>}]
 dyn_ankle_zipper_r_MSTR [{shapeType=Capsule, length=0.2, radius=0.2, rotation=<rot 90 0 0>}]
 dyn_coat_bk [{shapeType=Box, extentX=6.2674, extentY=10.2342, extentZ=1.042, center=<2.8901 0 -0.0001>}]
 dyn_coat_bk_MSTR [{shapeType=Capsule, length=0.2, radius=0.2, rotation=<rot 90 0 0>}]
 dyn_coat_bk_l [{shapeType=Box, extentX=6.2674, extentY=10.2342, extentZ=1.042, center=<2.8901 0 -0.0001>, rotation=<rot 0 0 15.0008>}]
 dyn_coat_bk_l_MSTR [{shapeType=Capsule, length=0.2, radius=0.2, rotation=<rot 90 0 0>}]
 dyn_coat_bk_r [{shapeType=Box, extentX=6.2674, extentY=10.2342, extentZ=1.042, center=<2.8901 0 -0.0001>, rotation=<rot 0 0 -15.0023>}]
 dyn_coat_bk_r_MSTR [{shapeType=Capsule, length=0.2, radius=0.2, rotation=<rot 90 0 0>}]
 dyn_coat_fr_l_01 [{shapeType=Box, extentX=6.2674, extentY=10.2342, extentZ=1.042, center=<2.8901 0 -0.0001>}]
 dyn_coat_fr_l_02 [{shapeType=Box, extentX=6.2674, extentY=10.2342, extentZ=1.042, center=<2.8901 0 -0.0001>}]
 dyn_coat_fr_l_MSTR [{shapeType=Capsule, length=0.2, radius=0.2, rotation=<rot 90 0 0>}]
 dyn_coat_fr_r_01 [{shapeType=Box, extentX=6.2674, extentY=10.2342, extentZ=1.042, center=<2.8901 0 -0.0001>}]
 dyn_coat_fr_r_02 [{shapeType=Box, extentX=6.2674, extentY=10.2342, extentZ=1.042, center=<2.8901 0 -0.0001>}]
 dyn_coat_fr_r_MSTR [{shapeType=Capsule, length=0.2, radius=0.2, rotation=<rot 90 0 0>}]
 dyn_coat_l [{shapeType=Box, extentX=6.2674, extentY=10.2342, extentZ=2.6045, center=<2.8901 0 -0.0001>}]
 dyn_coat_l_MSTR [{shapeType=Capsule, length=0.2, radius=0.2, rotation=<rot 90 0 0>}]
 dyn_coat_r [{shapeType=Box, extentX=6.2674, extentY=10.2342, extentZ=2.6045, center=<2.8901 0 -0.0001>}]
 dyn_coat_r_MSTR [{shapeType=Capsule, length=0.2, radius=0.2, rotation=<rot 90 0 0>}]
 dyn_collar_zipper [{shapeType=Box, extentX=1.7732, extentY=0.983, extentZ=0.5001, center=<0.7612 0 -0.0001>}]
 dyn_collar_zipper_MSTR [{shapeType=Capsule, length=0.2, radius=0.2, rotation=<rot 90 0 0>}]
 dyn_pouch_r [{shapeType=Box, extentX=14.5372, extentY=10.2342, extentZ=5.1974, center=<2.194 0.1835 1.4476>}]
 dyn_pouch_r_MSTR [{shapeType=Capsule, length=0.2, radius=0.2, rotation=<rot 90 0 0>}]
 dyn_zipper_l [{shapeType=Box, extentX=1.7732, extentY=0.983, extentZ=0.5001, center=<0.7612 0 -0.0001>}]
 dyn_zipper_l_MSTR [{shapeType=Capsule, length=0.2, radius=0.2, rotation=<rot 90 0 0>}]
 spine_04 [{shapeType=Capsule, length=2.1406, radius=10.5189, center=<-0.0003 2.9779 0>}]
 spine_05 [{shapeType=Capsule, length=6.974, radius=7.0898, center=<-2.4853 5.5786 0>}]
}
constraints [
 {bone1Name=dyn_collar_zipper, bone2Name=dyn_collar_zipper_MSTR, swing1LimitDegrees=45, swing2LimitDegrees=45, twistLimitDegrees=5}
 {bone1Name=dyn_coat_fr_l_01, bone2Name=dyn_coat_fr_l_MSTR, swing1LimitDegrees=45, swing2LimitDegrees=45, twistLimitDegrees=5}
 {bone1Name=dyn_coat_fr_l_02, bone2Name=dyn_coat_fr_l_01, swing1LimitDegrees=45, swing2LimitDegrees=45, twistLimitDegrees=5}
 {
  bone1Name=dyn_zipper_l_MSTR
  bone2Name=dyn_coat_fr_l_02
  swing1LimitDegrees=45
  swing1Motion=Locked
  swing2LimitDegrees=45
  swing2Motion=Locked
  twistLimitDegrees=45
  twistMotion=Locked
 }
 {bone1Name=dyn_zipper_l, bone2Name=dyn_zipper_l_MSTR, swing1LimitDegrees=45, swing2LimitDegrees=45, twistLimitDegrees=5}
 {bone1Name=dyn_coat_l, bone2Name=dyn_coat_l_MSTR, swing1LimitDegrees=45, swing2LimitDegrees=45, twistLimitDegrees=5}
 {bone1Name=dyn_coat_bk_l, bone2Name=dyn_coat_bk_l_MSTR, swing1LimitDegrees=45, swing2LimitDegrees=45, twistLimitDegrees=5}
 {bone1Name=dyn_coat_bk, bone2Name=dyn_coat_bk_MSTR, swing1LimitDegrees=45, swing2LimitDegrees=45, twistLimitDegrees=5}
 {bone1Name=dyn_coat_fr_r_01, bone2Name=dyn_coat_fr_r_MSTR, swing1LimitDegrees=45, swing2LimitDegrees=45, twistLimitDegrees=5}
 {bone1Name=dyn_coat_fr_r_02, bone2Name=dyn_coat_fr_r_01, swing1LimitDegrees=45, swing2LimitDegrees=45, twistLimitDegrees=5}
 {bone1Name=dyn_coat_r, bone2Name=dyn_coat_r_MSTR, swing1LimitDegrees=45, swing2LimitDegrees=45, twistLimitDegrees=5}
 {bone1Name=dyn_coat_bk_r, bone2Name=dyn_coat_bk_r_MSTR, swing1LimitDegrees=45, swing2LimitDegrees=45, twistLimitDegrees=5}
 {bone1Name=dyn_ankle_zipper_l, bone2Name=dyn_ankle_zipper_l_MSTR, swing1LimitDegrees=45, swing2LimitDegrees=45, twistLimitDegrees=5}
 {bone1Name=dyn_ankle_zipper_r, bone2Name=dyn_ankle_zipper_r_MSTR, swing1LimitDegrees=45, swing2LimitDegrees=45, twistLimitDegrees=5}
 {bone1Name=dyn_pouch_r, bone2Name=dyn_pouch_r_MSTR, swing1LimitDegrees=5, swing2LimitDegrees=5, twistLimitDegrees=5}
 {
  bone1Name=dyn_collar_zipper_MSTR
  bone2Name=spine_05
  swing1LimitDegrees=45
  swing1Motion=Free
  swing2LimitDegrees=45
  swing2Motion=Free
  twistLimitDegrees=45
  twistMotion=Free
 }
 {
  bone1Name=dyn_coat_fr_l_MSTR
  bone2Name=spine_04
  swing1LimitDegrees=45
  swing1Motion=Free
  swing2LimitDegrees=45
  swing2Motion=Free
  twistLimitDegrees=45
  twistMotion=Free
 }
 {
  bone1Name=dyn_coat_l_MSTR
  bone2Name=spine_04
  swing1LimitDegrees=45
  swing1Motion=Free
  swing2LimitDegrees=45
  swing2Motion=Free
  twistLimitDegrees=45
  twistMotion=Free
 }
 {
  bone1Name=dyn_coat_bk_l_MSTR
  bone2Name=spine_04
  swing1LimitDegrees=45
  swing1Motion=Free
  swing2LimitDegrees=45
  swing2Motion=Free
  twistLimitDegrees=45
  twistMotion=Free
 }
 {
  bone1Name=dyn_coat_bk_MSTR
  bone2Name=spine_04
  swing1LimitDegrees=45
  swing1Motion=Free
  swing2LimitDegrees=45
  swing2Motion=Free
  twistLimitDegrees=45
  twistMotion=Free
 }
 {
  bone1Name=dyn_coat_fr_r_MSTR
  bone2Name=spine_04
  swing1LimitDegrees=45
  swing1Motion=Free
  swing2LimitDegrees=45
  swing2Motion=Free
  twistLimitDegrees=45
  twistMotion=Free
 }
 {
  bone1Name=dyn_coat_r_MSTR
  bone2Name=spine_04
  swing1LimitDegrees=45
  swing1Motion=Free
  swing2LimitDegrees=45
  swing2Motion=Free
  twistLimitDegrees=45
  twistMotion=Free
 }
 {
  bone1Name=dyn_coat_bk_r_MSTR
  bone2Name=spine_04
  swing1LimitDegrees=45
  swing1Motion=Free
  swing2LimitDegrees=45
  swing2Motion=Free
  twistLimitDegrees=45
  twistMotion=Free
 }
 {
  bone1Name=spine_05
  bone2Name=spine_04
  swing1LimitDegrees=45
  swing1Motion=Free
  swing2LimitDegrees=45
  swing2Motion=Free
  twistLimitDegrees=45
  twistMotion=Free
 }
]
