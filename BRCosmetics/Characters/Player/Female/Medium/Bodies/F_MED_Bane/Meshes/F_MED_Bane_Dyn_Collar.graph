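# Physics asset for the dynamic collar on the female medium Bane body (F_MED_Bane) player cosmetic.
# Five collar segments (back, back-left, back-right, left, right) use thin box bodies, with wider boxes on the two sides.
# Each segment is constrained to a tiny capsule on its MSTR bone with matching swing and twist limits.
# An unconstrained head capsule provides a shape the collar pieces can collide against.
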PhysicsAsset /BRCosmetics/Characters/Player/Female/Medium/Bodies/F_MED_Bane/Meshes/F_MED_Bane_Dyn_Collar
bodies {
 dyn_collar_bk [{shapeType=Box, extentX=5.2279, extentY=10.2342, extentZ=1.042, center=<3.0836 0 -0.0001>}]
 dyn_collar_bk_MSTR [{shapeType=Capsule, length=0.2, radius=0.2, rotation=<rot 90 0 0>}]
 dyn_collar_bk_l [{shapeType=Box, extentX=5.2279, extentY=10.2342, extentZ=1.042, center=<3.0836 0 -0.0001>}]
 dyn_collar_bk_l_MSTR [{shapeType=Capsule, length=0.2, radius=0.2, rotation=<rot 90 0 0>}]
 dyn_collar_bk_r [{shapeType=Box, extentX=5.2279, extentY=10.2342, extentZ=1.042, center=<3.0836 0 -0.0001>}]
 dyn_collar_bk_r_MSTR [{shapeType=Capsule, length=0.2, radius=0.2, rotation=<rot 90 0 0>}]
 dyn_collar_l [{shapeType=Box, extentX=5.2279, extentY=17.9906, extentZ=1.042, center=<3.0836 0 -0.0001>}]
 dyn_collar_l_MSTR [{shapeType=Capsule, length=0.2, radius=0.2, rotation=<rot 90 0 0>}]
 dyn_collar_r [{shapeType=Box, extentX=5.2279, extentY=17.9906, extentZ=1.042, center=<3.0836 0 -0.0001>}]
 dyn_collar_r_MSTR [{shapeType=Capsule, length=0.2, radius=0.2, rotation=<rot 90 0 0>}]
 head [{shapeType=Capsule, length=5.5063, radius=4.6146, center=<-3.5009 -0.4241 0>, rotation=<rot 90 0 0>}]
}
constraints [
 {bone1Name=dyn_collar_bk, bone2Name=dyn_collar_bk_MSTR, swing1LimitDegrees=75, swing2LimitDegrees=75, twistLimitDegrees=75}
 {bone1Name=dyn_collar_bk_l, bone2Name=dyn_collar_bk_l_MSTR, swing1LimitDegrees=75, swing2LimitDegrees=75, twistLimitDegrees=75}
 {bone1Name=dyn_collar_l, bone2Name=dyn_collar_l_MSTR, swing1LimitDegrees=75, swing2LimitDegrees=75, twistLimitDegrees=75}
 {bone1Name=dyn_collar_bk_r, bone2Name=dyn_collar_bk_r_MSTR, swing1LimitDegrees=75, swing2LimitDegrees=75, twistLimitDegrees=75}
 {bone1Name=dyn_collar_r, bone2Name=dyn_collar_r_MSTR, swing1LimitDegrees=75, swing2LimitDegrees=75, twistLimitDegrees=75}
]
