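# Physics asset for the left-leg dynamic accessories on the F_MED_Bane female character body: a knife and an ankle zipper.
# Each accessory has a box body constrained to its own tiny capsule "MSTR" body, which serves as an anchor.
# Angular limits keep the knife to a narrow, asymmetric swing, while the zipper swings widely on both axes with little twist.
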PhysicsAsset /BRCosmetics/Characters/Player/Female/Medium/Bodies/F_MED_Bane/Meshes/F_MED_Bane_Dyn_leg_l
bodies {
 dyn_ankle_zipper_l [{shapeType=Box, extentX=1.7732, extentY=0.983, extentZ=0.5001, center=<0.6923 0 0.2146>, rotation=<rot 14.9999 0 0>}]
 dyn_ankle_zipper_l_MSTR [{shapeType=Capsule, length=0.2, radius=0.2, rotation=<rot 90 0 0>}]
 dyn_knife_l [{shapeType=Box, extentX=22.0184, extentY=4.4535, extentZ=2.0866, center=<3.1376 -0.2323 1.1979>, rotation=<rot 2.5053 -10.0114 -0.4421>}]
 dyn_knife_l_MSTR [{shapeType=Capsule, length=0.2, radius=0.2, rotation=<rot 90 0 0>}]
}
constraints [
 {bone1Name=dyn_knife_l, bone2Name=dyn_knife_l_MSTR, swing1LimitDegrees=15, swing2LimitDegrees=2, twistLimitDegrees=5}
 {bone1Name=dyn_ankle_zipper_l, bone2Name=dyn_ankle_zipper_l_MSTR, swing1LimitDegrees=45, swing2LimitDegrees=45, twistLimitDegrees=5}
]
