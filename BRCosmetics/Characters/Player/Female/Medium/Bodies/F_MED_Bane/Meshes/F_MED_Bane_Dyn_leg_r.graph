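# Physics asset for the dynamic accessories on the right leg of the F_MED_Bane female player body: an ankle zipper and five bullets.
# The zipper box is constrained to a tiny master capsule, with wider swing than twist.
# The five bullet capsules each hang in parallel, not chained, off one shared master with tight angular limits.
# That bullet master joins the thigh twist body with free swing and twist, while the zipper master has no parent constraint.
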PhysicsAsset /BRCosmetics/Characters/Player/Female/Medium/Bodies/F_MED_Bane/Meshes/F_MED_Bane_Dyn_leg_r
bodies {
 dyn_ankle_zipper_r [{shapeType=Box, extentX=1.7732, extentY=0.983, extentZ=0.5001, center=<0.6923 0 0.2146>, rotation=<rot 14.9999 0 0>}]
 dyn_ankle_zipper_r_MSTR [{shapeType=Capsule, length=0.2, radius=0.2, rotation=<rot 90 0 0>}]
 dyn_bullet_r_01 [{shapeType=Capsule, length=8.0743, radius=1.2047, center=<0 0 1.312>, rotation=<rot 90 0 0>}]
 dyn_bullet_r_01_MSTR [{shapeType=Capsule, length=0.2, radius=0.2, rotation=<rot 90 0 0>}]
 dyn_bullet_r_02 [{shapeType=Capsule, length=8.0743, radius=1.2047, center=<0 0 1.312>, rotation=<rot 90 0 0>}]
 dyn_bullet_r_03 [{shapeType=Capsule, length=8.0743, radius=1.2047, center=<0 0 1.312>, rotation=<rot 90 0 0>}]
 dyn_bullet_r_04 [{shapeType=Capsule, length=8.0743, radius=1.2047, center=<0 0 1.312>, rotation=<rot 90 0 0>}]
 dyn_bullet_r_05 [{shapeType=Capsule, length=8.0743, radius=1.2047, center=<0 0 1.312>, rotation=<rot 90 0 0>}]
 thigh_twist_01_r [{shapeType=Capsule, length=2, radius=2, rotation=<rot 90 0 0>}]
}
constraints [
 {bone1Name=dyn_ankle_zipper_r, bone2Name=dyn_ankle_zipper_r_MSTR, swing1LimitDegrees=45, swing2LimitDegrees=45, twistLimitDegrees=5}
 {
  bone1Name=dyn_bullet_r_01_MSTR
  bone2Name=thigh_twist_01_r
  swing1LimitDegrees=45
  swing1Motion=Free
  swing2LimitDegrees=45
  swing2Motion=Free
  twistLimitDegrees=45
  twistMotion=Free
 }
 {bone1Name=dyn_bullet_r_01, bone2Name=dyn_bullet_r_01_MSTR, swing1LimitDegrees=10, swing2LimitDegrees=10, twistLimitDegrees=5}
 {bone1Name=dyn_bullet_r_02, bone2Name=dyn_bullet_r_01_MSTR, swing1LimitDegrees=10, swing2LimitDegrees=10, twistLimitDegrees=5}
 {bone1Name=dyn_bullet_r_03, bone2Name=dyn_bullet_r_01_MSTR, swing1LimitDegrees=10, swing2LimitDegrees=10, twistLimitDegrees=5}
 {bone1Name=dyn_bullet_r_04, bone2Name=dyn_bullet_r_01_MSTR, swing1LimitDegrees=10, swing2LimitDegrees=10, twistLimitDegrees=5}
 {bone1Name=dyn_bullet_r_05, bone2Name=dyn_bullet_r_01_MSTR, swing1LimitDegrees=10, swing2LimitDegrees=10, twistLimitDegrees=5}
]
