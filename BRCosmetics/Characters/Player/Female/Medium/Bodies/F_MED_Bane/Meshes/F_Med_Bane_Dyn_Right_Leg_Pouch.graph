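# Physics asset for the dynamic right leg pouch on the F_Med_Bane female body cosmetic, letting the pouch move as secondary physics.
# It has two same-sized box bodies, one on the dyn_pouch_r bone and one on its dyn_pouch_r_MSTR counterpart, with slightly offset centers.
# A single constraint joins them, allowing a modest swing on one axis while keeping the other swing and twist almost rigid.
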PhysicsAsset /BRCosmetics/Characters/Player/Female/Medium/Bodies/F_MED_Bane/Meshes/F_Med_Bane_Dyn_Right_Leg_Pouch
bodies {
 dyn_pouch_r [{shapeType=Box, extentX=17.5, extentY=11.5, extentZ=7.5, shapeName=dyn_pouch_r_box, center=<2.6178 0.2158 1.9035>}]
 dyn_pouch_r_MSTR [{shapeType=Box, extentX=17.5, extentY=11.5, extentZ=7.5, shapeName=dyn_pouch_r_MSTR_box, center=<2.6178 0.2158 2.3258>}]
}
constraints [{bone1Name=dyn_pouch_r, bone2Name=dyn_pouch_r_MSTR, swing1LimitDegrees=10, swing2LimitDegrees=1, twistLimitDegrees=1}]
