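# Post-process style AnimBlueprint for the F_MED_Bane cosmetic body, copying the attached parent mesh's pose and adding corrective and secondary motion.
# Weighted constraints place kneepad, zippers and gear, while thigh and arm angles from the EventGraph drive additive pelvis and collar correctives.
# Five RigidBody nodes with separate physics assets, then a sleeve AnimDynamics, simulate dangling parts.
# The alt foot/ball constraint chain is not connected to the output.
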AnimBlueprint /BRCosmetics/Characters/Player/Female/Medium/Bodies/F_MED_Bane/Meshes/F_MED_Bane_AnimBP
target_skeleton=&/F_MED_Bane_Skeleton
graphs [
 AnimGraph {nodes [
   n1 = "Constraint - alt_foot_l (R)"(Alpha=1, ConstraintWeights_0=1) {
    alphaBoolBlend {blendOption=Linear}
    boneToModify {boneName=alt_foot_l}
    constraintSetup [{transformType=Rotation, targetBone {boneName=foot_l}}]
    constraintWeights [1]
   }
   n2 = "Constraint - alt_ball_l (R)"(ComponentPose=@n1, Alpha=1, ConstraintWeights_0=1) {
    alphaBoolBlend {blendOption=Linear}
    boneToModify {boneName=alt_ball_l}
    constraintSetup [{transformType=Rotation, targetBone {boneName=ball_l}}]
    constraintWeights [1]
   }
   n3 = "Constraint - alt_foot_r (R)"(ComponentPose=@n2, Alpha=1, ConstraintWeights_0=1) {
    alphaBoolBlend {blendOption=Linear}
    boneToModify {boneName=alt_foot_r}
    constraintSetup [{transformType=Rotation, targetBone {boneName=foot_r}}]
    constraintWeights [1]
   }
   "Constraint - alt_ball_r (R)"(ComponentPose=@n3, Alpha=1, ConstraintWeights_0=1) {
    alphaBoolBlend {blendOption=Linear}
    boneToModify {boneName=alt_ball_r}
    constraintSetup [{transformType=Rotation, targetBone {boneName=ball_r}}]
    constraintWeights [1]
   }
   n4 = "Constraint - kneepad_r (PP)"!(
    ComponentPose=Local To Component(LocalPose=Copy Pose From Mesh() {bUseAttachedParent=true})
    Alpha=1
    ConstraintWeights_0=0.3263
    ConstraintWeights_1:float=0.6737
   ) -> (Pose) {
    alphaBoolBlend {blendOption=Linear}
    boneToModify {boneName=kneepad_r}
    constraintSetup [{targetBone {boneName=thigh_r}}, {targetBone {boneName=calf_r}}]
    constraintWeights [1,1]
   }
   n5 = "Constraint - kneepad_r (RR)"!(ComponentPose=@n4, Alpha=1, ConstraintWeights_0=0.3, ConstraintWeights_1:float=0.7) -> (Pose) {
    alphaBoolBlend {blendOption=Linear}
    boneToModify {boneName=kneepad_r}
    constraintSetup [{transformType=Rotation, targetBone {boneName=thigh_r}}, {transformType=Rotation, targetBone {boneName=calf_r}}]
    constraintWeights [1,1]
   }
   n6 = Get thigh_rot_l()
   n7 = "Transform (Modify) Bone - Bone: pelvis_l"(ComponentPose=@n5, Alpha=@n6, Translation="-6.240,-1.520,1.245", Rotation="0,0,0", Scale="1,1,1") {
    scaleMode=BMM_Additive
    translationMode=BMM_Additive
    translationSpace=BCS_ParentBoneSpace
    alphaBoolBlend {blendOption=Linear}
    boneToModify {boneName=pelvis_l}
    translation=<-6.24 -1.52 1.245>
   }
   n8 = "Transform (Modify) Bone - Bone: pelvis_mid_l"(ComponentPose=@n7, Alpha=@n6, Translation="-1.450,-1.138,0.518", Rotation="0,0,0", Scale="1,1,1") {
    scaleMode=BMM_Additive
    translationMode=BMM_Additive
    translationSpace=BCS_ParentBoneSpace
    alphaBoolBlend {blendOption=Linear}
    boneToModify {boneName=pelvis_mid_l}
    translation=<-1.45 -1.138 0.518>
   }
   n9 = "Transform (Modify) Bone - Bone: pelvis_side_l"(ComponentPose=@n8, Alpha=@n6, Translation="0.505,-0.560,-1.440", Rotation="0,0,0", Scale="1,1,1") {
    scaleMode=BMM_Additive
    translationMode=BMM_Additive
    translationSpace=BCS_ParentBoneSpace
    alphaBoolBlend {blendOption=Linear}
    boneToModify {boneName=pelvis_side_l}
    translation=<0.505 -0.56 -1.44>
   }
   n10 = Get thigh_rot_r()
   n11 = "Transform (Modify) Bone - Bone: pelvis_r"(ComponentPose=@n9, Alpha=@n10, Translation="-1.106,-0.155,0.183", Rotation="0,0,0", Scale="1,1,1") {
    scaleMode=BMM_Additive
    translationMode=BMM_Additive
    translationSpace=BCS_ParentBoneSpace
    alphaBoolBlend {blendOption=Linear}
    boneToModify {boneName=pelvis_r}
    translation=<-1.106 -0.155 0.183>
   }
   n12 = "Transform (Modify) Bone - Bone: pelvis_mid_r"(ComponentPose=@n11, Alpha=@n10, Translation="-0.848,0.002,0.106", Rotation="0,0,0", Scale="1,1,1") {
    scaleMode=BMM_Additive
    translationMode=BMM_Additive
    translationSpace=BCS_ParentBoneSpace
    alphaBoolBlend {blendOption=Linear}
    boneToModify {boneName=pelvis_mid_r}
    translation=<-0.848 0.002 0.106>
   }
   n13 = "Transform (Modify) Bone - Bone: pelvis_side_r"(ComponentPose=@n12, Alpha=@n10, Translation="-0.144,-0.340,-1.135", Rotation="0,0,0", Scale="1,1,1") {
    scaleMode=BMM_Additive
    translationMode=BMM_Additive
    translationSpace=BCS_ParentBoneSpace
    alphaBoolBlend {blendOption=Linear}
    boneToModify {boneName=pelvis_side_r}
    translation=<-0.144 -0.34 -1.135>
   }
   n14 = "Transform (Modify) Bone - Bone: dyn_collar_r_MSTR"(ComponentPose=@n13, Alpha=Get arm_up_r() -> (arm_up_r:double), Translation="0,0,0", Rotation="-14.942,-5.175,1.338", Scale="1,1,1") {
    rotationMode=BMM_Additive
    rotationSpace=BCS_BoneSpace
    translationMode=BMM_Additive
    translationSpace=BCS_BoneSpace
    alphaBoolBlend {blendOption=Linear}
    alphaScaleBiasClamp {bInterpResult=true, interpSpeedDecreasing=15, interpSpeedIncreasing=15}
    boneToModify {boneName=dyn_collar_r_MSTR}
    rotation=<rot -14.942 -5.1752 1.338>
   }
   n15 = "Constraint - dyn_collar_zipper_MSTR (PPPPPPP)"!(
    ComponentPose=@n14
    Alpha=1
    ConstraintWeights_0=0.0454
    ConstraintWeights_1:float=0
    ConstraintWeights_2:float=0.3905
    ConstraintWeights_3:float=0.0981
    ConstraintWeights_4:float=0
    ConstraintWeights_5:float=0.2883
    ConstraintWeights_6:float=0.1777
   ) -> (Pose) {
    alphaBoolBlend {blendOption=Linear}
    boneToModify {boneName=dyn_collar_zipper_MSTR}
    constraintSetup [
     {targetBone {boneName=head}}
     {targetBone {boneName=clavicle_l}}
     {targetBone {boneName=spine_05}}
     {targetBone {boneName=C_jaw}}
     {targetBone {boneName=clavicle_r}}
     {targetBone {boneName=neck_02}}
     {targetBone {boneName=neck_01}}
    ]
    constraintWeights [1,1,1,1,1,1,1]
   }
   n16 = "Constraint - dyn_zipper_l_MSTR (P)"!(ComponentPose=@n15, Alpha=1, ConstraintWeights_0=1, ConstraintWeights_1:float=0) -> (Pose) {
    alphaBoolBlend {blendOption=Linear}
    boneToModify {boneName=dyn_zipper_l_MSTR}
    constraintSetup [{targetBone {boneName=dyn_coat_fr_l_02}}]
    constraintWeights [1,0]
   }
   n17 = "Constraint - dyn_ankle_zipper_l_MSTR (PPPP)"!(
    ComponentPose=@n16
    Alpha=1
    ConstraintWeights_0=0.0466
    ConstraintWeights_1:float=0.73
    ConstraintWeights_2:float=0.0948
    ConstraintWeights_3:float=0.1286
   ) -> (Pose) {
    alphaBoolBlend {blendOption=Linear}
    boneToModify {boneName=dyn_ankle_zipper_l_MSTR}
    constraintSetup [
     {targetBone {boneName=calf_l}}
     {targetBone {boneName=calf_twist_01_l}}
     {targetBone {boneName=alt_foot_l}}
     {targetBone {boneName=calf_twist_02_l}}
    ]
    constraintWeights [1,1,1,1]
   }
   n18 = "Constraint - dyn_ankle_zipper_r_MSTR (PPPP)"!(
    ComponentPose=@n17
    Alpha=1
    ConstraintWeights_0=0.1286
    ConstraintWeights_1:float=0.73
    ConstraintWeights_2:float=0.0948
    ConstraintWeights_3:float=0.0466
   ) -> (Pose) {
    alphaBoolBlend {blendOption=Linear}
    boneToModify {boneName=dyn_ankle_zipper_r_MSTR}
    constraintSetup [
     {targetBone {boneName=calf_twist_02_r}}
     {targetBone {boneName=calf_twist_01_r}}
     {targetBone {boneName=alt_foot_r}}
     {targetBone {boneName=calf_r}}
    ]
    constraintWeights [1,1,1,1]
   }
   n19 = "Constraint - dyn_pouch_r_MSTR (PPP)"!(ComponentPose=@n18, Alpha=1, ConstraintWeights_0=0.0126, ConstraintWeights_1:float=0.6663, ConstraintWeights_2:float=0.3211) -> (Pose) {
    alphaBoolBlend {blendOption=Linear}
    boneToModify {boneName=dyn_pouch_r_MSTR}
    constraintSetup [{targetBone {boneName=thigh_r}}, {targetBone {boneName=pelvis}}, {targetBone {boneName=thigh_twist_01_r}}]
    constraintWeights [1,1,1]
   }
   n20 = "Constraint - dyn_bullet_r_01_MSTR (PPP)"!(ComponentPose=@n19, Alpha=1, ConstraintWeights_0=0.6686, ConstraintWeights_1:float=0.0127, ConstraintWeights_2:float=0.3187) -> (Pose) {
    alphaBoolBlend {blendOption=Linear}
    boneToModify {boneName=dyn_bullet_r_01_MSTR}
    constraintSetup [{targetBone {boneName=thigh_r}}, {targetBone {boneName=pelvis}}, {targetBone {boneName=thigh_twist_01_r}}]
    constraintWeights [1,1,1]
   }
   n21 = "Constraint - dyn_knife_l_MSTR (PP)"!(ComponentPose=@n20, Alpha=1, ConstraintWeights_0=0.1781, ConstraintWeights_1:float=0.8219) -> (Pose) {
    alphaBoolBlend {blendOption=Linear}
    boneToModify {boneName=dyn_knife_l_MSTR}
    constraintSetup [{targetBone {boneName=calf_l}}, {targetBone {boneName=calf_twist_02_l}}]
    constraintWeights [1,1]
   }
   n22 = RigidBody(ComponentPose=@n21) {
    bClampLinearTranslationLimitToRefPose=true
    bOverrideWorldGravity=true
    overridePhysicsAsset=&/F_MED_Bane_Dyn
    simulationSpace=BaseBoneSpace
    alphaBoolBlend {blendOption=Linear}
    baseBoneRef {boneName=spine_03}
    componentAppliedLinearAccClamp=<3000 3000 3000>
    componentLinearAccScale=<0.4 0.3 0.3>
    componentLinearVelScale=<0.4 0.3 0.3>
    overrideWorldGravity {z=-980}
   }
   n23 = RigidBody(ComponentPose=@n22) {
    bClampLinearTranslationLimitToRefPose=true
    bOverrideWorldGravity=true
    overridePhysicsAsset=&/F_MED_Bane_Dyn_Collar
    simulationSpace=BaseBoneSpace
    alphaBoolBlend {blendOption=Linear}
    baseBoneRef {boneName=spine_05}
    componentAppliedLinearAccClamp=<3000 3000 3000>
    componentLinearAccScale=<0.4 0.3 0.3>
    componentLinearVelScale=<0.4 0.3 0.3>
    overrideWorldGravity {z=-980}
   }
   n24 = RigidBody(ComponentPose=@n23) {
    bClampLinearTranslationLimitToRefPose=true
    bOverrideWorldGravity=true
    overridePhysicsAsset=&/F_MED_Bane_Dyn_leg_l
    simulationSpace=BaseBoneSpace
    alphaBoolBlend {blendOption=Linear}
    baseBoneRef {boneName=calf_l}
    componentAppliedLinearAccClamp=<3000 3000 3000>
    componentLinearAccScale=<0.6 0.5 0.5>
    componentLinearVelScale=<0.6 0.5 0.5>
    overrideWorldGravity {z=-980}
   }
   n25 = RigidBody(ComponentPose=@n24) {
    bClampLinearTranslationLimitToRefPose=true
    bOverrideWorldGravity=true
    overridePhysicsAsset=&/F_MED_Bane_Dyn_leg_r
    simulationSpace=BaseBoneSpace
    alphaBoolBlend {blendOption=Linear}
    baseBoneRef {boneName=thigh_r}
    componentAppliedLinearAccClamp=<3000 3000 3000>
    componentLinearAccScale=<0.4 0.3 0.3>
    componentLinearVelScale=<0.4 0.3 0.3>
    overrideWorldGravity {z=-980}
   }
   n26 = RigidBody(ComponentPose=@n25) {
    bClampLinearTranslationLimitToRefPose=true
    bOverrideWorldGravity=true
    overridePhysicsAsset=&/F_Med_Bane_Dyn_Right_Leg_Pouch
    simulationSpace=BaseBoneSpace
    alphaBoolBlend {blendOption=Linear}
    baseBoneRef {boneName=thigh_r}
    componentLinearAccScale=<0.3 0.3 0.45>
    componentLinearVelScale=<0.15 0.15 0.225>
    overrideWorldGravity {z=-980}
   }
   n27 = Anim Dynamics - Bone: dyn_sleeve_l(ComponentPose=@n26, Alpha=0.3, ExternalForce="0,0,0") {
    angularDampingOverride=1
    angularSpringConstant=1100
    bAngularSpring=true
    bGravityOverrideInSimSpace=true
    bOverrideAngularDamping=true
    bUseGravityOverride=true
    gravityScale=0.35
    simulationSpace=RootRelative
    alphaBoolBlend {blendOption=Linear}
    boundBone {boneName=dyn_sleeve_l}
    componentAppliedLinearAccClamp=<3000 3000 3000>
    componentLinearAccScale=<0.4 0.3 0.3>
    componentLinearVelScale=<0.4 0.3 0.3>
    gravityOverride {z=-300}
    physicsBodyDefinitions [{
      sphereCollisionRadius=0
      boundBone {boneName=dyn_sleeve_l}
      boxExtents=<10 3 3>
      constraintSetup {angularLimitsMax=<0 3 3>, angularLimitsMin=<0 -3 -3>, angularTarget=<1 0 0>}
      localJointOffset=<-4 0 0>
     }]
   }
   Output Pose(Result=Component To Local(ComponentPose=@n27)) {layerGroup=None}
  ]}
 EventGraph {nodes [
   Event BlueprintUpdateAnimation()
   n1 = BreakTransform(InTransform=GetDeltaTransformFromRefPose(self=GetOwnerSkeletalMeshComponent(), BoneName=thigh_r, BaseName=pelvis))
   > Set thigh_rot_r(thigh_rot_r=MapRangeClamped(Value=BreakRotator(InRot=@n1.Rotation).Yaw, InRangeA=0, InRangeB=-90))
   n2 = BreakTransform(InTransform=GetDeltaTransformFromRefPose(self=GetOwnerSkeletalMeshComponent(), BoneName=thigh_l, BaseName=pelvis))
   > Set thigh_rot_l(thigh_rot_l=MapRangeClamped(Value=BreakRotator(InRot=@n2.Rotation).Yaw, InRangeA=-15, InRangeB=-110))
   n3 = BreakTransform(InTransform=GetDeltaTransformFromRefPose(self=GetOwnerSkeletalMeshComponent(), BoneName=upperarm_r, BaseName=spine_05))
   > Set arm_up_r(arm_up_r:double=MapRangeClamped(Value=BreakRotator(InRot=@n3.Rotation).Yaw, InRangeA=110, InRangeB=130)) -> (Output_Get:double)
  ]}
]
variables [
 arm_up_r {type=number}
 thigh_rot_l {type=number}
 thigh_rot_r {type=number}
]
