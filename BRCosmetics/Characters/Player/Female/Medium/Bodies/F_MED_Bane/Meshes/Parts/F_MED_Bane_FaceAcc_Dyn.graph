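# Physics asset for the F_MED_Bane face accessory part, setting up dynamic hair on the female player character.
# Six jointed hair chains (front, left, right, three at back) hang from the head, using mostly boxes, some capsules, and narrow twist limits.
# Capsules form the head, neck and upper spine, joined by free-rotation constraints.
# Cross-constraints tie both back side chains to the center back chain, with free rotation on the right link.
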PhysicsAsset /BRCosmetics/Characters/Player/Female/Medium/Bodies/F_MED_Bane/Meshes/Parts/F_MED_Bane_FaceAcc_Dyn
bodies {
 dyn_hair_bk_01 [{shapeType=Box, extentX=6.1017, extentY=7.5, extentZ=1.7408, center=<3.3857 0 0>}]
 dyn_hair_bk_02 [{shapeType=Box, extentX=6.1017, extentY=12.3439, extentZ=2.1269, center=<3.3857 0 0>}]
 dyn_hair_bk_03 [{shapeType=Box, extentX=6.1017, extentY=12.3439, extentZ=2.1269, center=<3.3857 0 0>}]
 dyn_hair_bk_l_01 [{shapeType=Box, extentX=6.1017, extentY=7.5, extentZ=0.856, center=<3.3857 0 0>}]
 dyn_hair_bk_l_02 [{shapeType=Box, extentX=6.1017, extentY=14.4821, extentZ=0.856, center=<3.3857 0 0>}]
 dyn_hair_bk_l_03 [{shapeType=Box, extentX=6.1017, extentY=14.4821, extentZ=0.856, center=<3.3857 0 0>}]
 dyn_hair_bk_r_01 [{shapeType=Box, extentX=6.1017, extentY=7.5, extentZ=2.2289, center=<3.3857 0 0>}]
 dyn_hair_bk_r_02 [{shapeType=Box, extentX=6.1017, extentY=7.5, extentZ=0.856, center=<3.3857 0 0>}]
 dyn_hair_bk_r_03 [{shapeType=Box, extentX=6.1017, extentY=14.3569, extentZ=0.856, center=<3.3857 0 0>}]
 dyn_hair_bk_r_04 [{shapeType=Capsule, length=12, radius=1.5, shapeName=dyn_hair_bk_r_04_capsule, center=<2.5 0 0>, rotation=<rot 0 0 90>}]
 dyn_hair_fr_01 [{shapeType=Capsule, length=4, radius=1.5, shapeName=dyn_hair_fr_01_capsule, center=<3 0 0>, rotation=<rot 90 0 0>}]
 dyn_hair_fr_02 [{shapeType=Capsule, length=4, radius=1.5, shapeName=dyn_hair_fr_02_capsule, center=<3 0 0>, rotation=<rot 90 0 0>}]
 dyn_hair_l_01 [{shapeType=Box, extentX=6.1017, extentY=4.794, extentZ=0.856, center=<3.2306 0 0>}]
 dyn_hair_l_02 [{shapeType=Capsule, length=2, radius=2, shapeName=dyn_hair_l_02_capsule, center=<3 0.5 0>, rotation=<rot 90 0 0>}]
 dyn_hair_l_03 [{shapeType=Capsule, length=2, radius=1.5, shapeName=dyn_hair_l_03_capsule, center=<2 1 0>, rotation=<rot 90 0 0>}]
 dyn_hair_r_01 [{shapeType=Box, extentX=3.9772, extentY=7.5, extentZ=0.856, center=<4.0626 0.0001 0>}]
 dyn_hair_r_02 [{shapeType=Box, extentX=6.1017, extentY=10.2587, extentZ=0.856, center=<3.3857 0 0>}]
 dyn_hair_r_03 [{shapeType=Box, extentX=6.1017, extentY=10.2587, extentZ=0.856, center=<3.3857 0 0>}]
 head [
  {shapeType=Capsule, length=5.3867, radius=6.6133, center=<3.053 0.2123 0.1132>, rotation=<rot 90 0 0>}
  {shapeType=Capsule, length=5.3867, radius=6.6133, center=<3.0531 3.7709 0.1132>, rotation=<rot 90 0 0>}
  {shapeType=Capsule, length=2.8757, radius=1.2271, center=<2.3942 -0.6886 -6.2403>, rotation=<rot 85.1399 0 0>}
  {shapeType=Capsule, length=2.8757, radius=1.2271, center=<1.9115 -0.7859 6.348>, rotation=<rot 80.2175 180 180>}
 ]
 neck_01 [{shapeType=Capsule, length=2.1545, radius=8.4227, center=<2.1336 0.9263 0.0001>, rotation=<rot 90 0 13.5236>}]
 spine_05 [{shapeType=Capsule, length=10, radius=10, center=<2.3756 -0.0002 0>}]
}
constraints [
 {bone1Name=dyn_hair_bk_l_01, bone2Name=head, swing1LimitDegrees=45, swing2LimitDegrees=45, twistLimitDegrees=2}
 {bone1Name=dyn_hair_bk_l_02, bone2Name=dyn_hair_bk_l_01, swing1LimitDegrees=45, swing2LimitDegrees=45, twistLimitDegrees=2}
 {bone1Name=dyn_hair_bk_l_03, bone2Name=dyn_hair_bk_l_02, swing1LimitDegrees=45, swing2LimitDegrees=45, twistLimitDegrees=2}
 {bone1Name=dyn_hair_bk_01, bone2Name=head, swing1LimitDegrees=45, swing2LimitDegrees=45, twistLimitDegrees=5}
 {bone1Name=dyn_hair_bk_02, bone2Name=dyn_hair_bk_01, swing1LimitDegrees=45, swing2LimitDegrees=45, twistLimitDegrees=5}
 {bone1Name=dyn_hair_bk_03, bone2Name=dyn_hair_bk_02, swing1LimitDegrees=45, swing2LimitDegrees=45, twistLimitDegrees=5}
 {bone1Name=dyn_hair_l_01, bone2Name=head, swing1LimitDegrees=15, swing2LimitDegrees=15, twistLimitDegrees=2}
 {bone1Name=dyn_hair_l_02, bone2Name=dyn_hair_l_01, swing1LimitDegrees=45, swing2LimitDegrees=45, twistLimitDegrees=5}
 {bone1Name=dyn_hair_l_03, bone2Name=dyn_hair_l_02, swing1LimitDegrees=45, swing2LimitDegrees=45, twistLimitDegrees=5}
 {bone1Name=dyn_hair_fr_01, bone2Name=head, swing1LimitDegrees=45, swing2LimitDegrees=45, twistLimitDegrees=5}
 {bone1Name=dyn_hair_fr_02, bone2Name=dyn_hair_fr_01, swing1LimitDegrees=45, swing2LimitDegrees=45, twistLimitDegrees=5}
 {bone1Name=dyn_hair_r_01, bone2Name=head, swing1LimitDegrees=45, swing2LimitDegrees=10, twistLimitDegrees=5}
 {bone1Name=dyn_hair_r_02, bone2Name=dyn_hair_r_01, swing1LimitDegrees=45, swing2LimitDegrees=45, twistLimitDegrees=5}
 {bone1Name=dyn_hair_r_03, bone2Name=dyn_hair_r_02, swing1LimitDegrees=45, swing2LimitDegrees=45, twistLimitDegrees=5}
 {bone1Name=dyn_hair_bk_r_01, bone2Name=head, swing1LimitDegrees=45, swing2LimitDegrees=45, twistLimitDegrees=5}
 {bone1Name=dyn_hair_bk_r_02, bone2Name=dyn_hair_bk_r_01, swing1LimitDegrees=45, swing2LimitDegrees=45, twistLimitDegrees=5}
 {bone1Name=dyn_hair_bk_r_03, bone2Name=dyn_hair_bk_r_02, swing1LimitDegrees=45, swing2LimitDegrees=45, twistLimitDegrees=5}
 {bone1Name=dyn_hair_bk_r_04, bone2Name=dyn_hair_bk_r_03, swing1LimitDegrees=45, swing2LimitDegrees=45, twistLimitDegrees=5}
 {
  bone1Name=head
  bone2Name=neck_01
  swing1LimitDegrees=45
  swing1Motion=Free
  swing2LimitDegrees=45
  swing2Motion=Free
  twistLimitDegrees=45
  twistMotion=Free
 }
 {
  bone1Name=neck_01
  bone2Name=spine_05
  swing1LimitDegrees=45
  swing1Motion=Free
  swing2LimitDegrees=45
  swing2Motion=Free
  twistLimitDegrees=45
  twistMotion=Free
 }
 {
  bone1Name=dyn_hair_bk_r_03
  bone2Name=dyn_hair_bk_02
  swing1LimitDegrees=45
  swing1Motion=Free
  swing2LimitDegrees=45
  swing2Motion=Free
  twistLimitDegrees=45
  twistMotion=Free
 }
 {bone1Name=dyn_hair_bk_l_02, bone2Name=dyn_hair_bk_02, swing1LimitDegrees=45, swing2LimitDegrees=45, twistLimitDegrees=2}
]
